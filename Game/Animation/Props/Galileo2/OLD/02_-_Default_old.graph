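# Placeholder-style Default Lit material in an OLD folder for the Galileo2 animation prop, named like an imported default material.
# Its graph holds a single dark grey vector parameter, with no connection to any material output shown.
# Usage flags enable it for skeletal meshes, mesh deformers and Nanite.
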Material /Game/Animation/Props/Galileo2/OLD/02_-_Default_old
graph {
 properties {
  bUsedWithMeshDeformer=true
  bUsedWithNanite=true
  bUsedWithSkeletalMesh=true
  pixelDepthOffsetMode=PDOM_Legacy
  usedShadingModels=Default Lit
 }
 nodes [VectorParameter() {defaultValue=<rgba 0.2216 0.2216 0.2216 0>}]
}
